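# M_TreeLight is a material for the TreeLight animation prop, built on the shared character material function chain.
# BaseFN_CharacterFunction_PreFX supplies the base attributes, which MF_BlinkLight from the weapon skin library modifies using the SkinFX_Mask texture parameter.
# The result passes through the character PostFX function, then CINE_MF_CharPostFX.
# It uses material attributes and an expression-driven shading model covering Default Lit and Subsurface Profile.
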Material /BRCosmetics/Animation/Props/TreeLight/Materials/M_TreeLight
graph {
 properties {
  bAutomaticallySetUsageInEditor=false
  bUseMaterialAttributes=true
  pixelDepthOffsetMode=PDOM_Legacy
  refractionMethod=RM_IndexOfRefraction
  shadingModel=MSM_FromMaterialExpression
  subsurfaceProfile=&SS_HeroSkin_02
  usedShadingModels=Default Lit | Subsurface Profile
 }
 nodes [
  n1 = MaterialFunctionCall() {materialFunction=&BaseFN_CharacterFunction_PreFX}
  n2 = TextureSample(Tex=TextureObjectParameter() {group=Skin FX, parameterName=SkinFX_Mask})
  n3 = MaterialFunctionCall(
   AO=@n1.AO
   Distance Blend=@n1.AO
   Fresnel=@n1.AO
   Gmap=@n1.AO
   Pre Skinned Local Position=@n1.AO
   Material=MaterialFunctionCall!(FX_Mask_Vectoir=@n2.RGBA, MA_Input=@n1.Result, FX_MaskInput=@n2.RGBA) {materialFunction=/Game/Weapons/WeaponSkins/MaterialFunction/MF_BlinkLight}.MA_Output
  ) {materialFunction=&BaseFN_CharacterFunction_PostFX}
  MaterialFunctionCall(In=@n3.Result) {materialFunction=&CINE_MF_CharPostFX}
 ]
}
